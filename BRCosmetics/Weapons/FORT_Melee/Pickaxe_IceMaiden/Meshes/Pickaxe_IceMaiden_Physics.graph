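# Physics asset for the IceMaiden pickaxe, providing capsule bodies and joints for its handle and dynamic heart and bow bones.
# The heart joins the two-handed handle with free swing and twist, and anchors six bow chains (left and right A, B, C).
# Each chain has two linked capsules, except left C with three.
# Bow joints set per-chain swing limits and a tight twist limit, restricting how far the pieces bend and roll.
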PhysicsAsset /BRCosmetics/Weapons/FORT_Melee/Pickaxe_IceMaiden/Meshes/Pickaxe_IceMaiden_Physics
bodies {
 Melee_TwoHanded_Handle [{shapeType=Capsule, length=81.6465, radius=1.5452, center=<0.9116 -0.8429 35.3362>, rotation=<rot 0.102 -90 0.1915>}]
 dyn_heart [{shapeType=Capsule, length=2.005, radius=3.005, center=<0 0.7649 0>, rotation=<rot -90 26.565 -116.5633>}]
 dyn_left_bowA_01 [{shapeType=Capsule, length=2.2995, radius=1.9427, center=<3 0 0>, rotation=<rot 0 0 90>}]
 dyn_left_bowA_02 [{shapeType=Capsule, length=2.5811, radius=3.0007, center=<3 0 0>, rotation=<rot 0 0 90>}]
 dyn_left_bowB_01 [{shapeType=Capsule, length=2.0495, radius=0.9427, center=<1.5 0 0>, rotation=<rot 0 0 90>}]
 dyn_left_bowB_02 [{shapeType=Capsule, length=2.3311, radius=2.0007, center=<1.5 0 0>, rotation=<rot 0 0 90>}]
 dyn_left_bowC_01 [{shapeType=Capsule, length=1.981, radius=1.4994, center=<3.8481 0 0>, rotation=<rot -90 -0.02 0.0202>}]
 dyn_left_bowC_02 [{shapeType=Capsule, length=2.5778, radius=2.4027, center=<3.3323 0 0>, rotation=<rot -90 17.354 -13.5677>}]
 dyn_left_bowC_03 [{shapeType=Capsule, length=2.6113, radius=2.8428, center=<4 0 0>, rotation=<rot 0 24.9996 90>}]
 dyn_right_bowA_01 [{shapeType=Capsule, length=2.2995, radius=1.9427, center=<3 0 0>, rotation=<rot 0 0 90>}]
 dyn_right_bowA_02 [{shapeType=Capsule, length=2.5811, radius=3.0007, center=<3 0 0>, rotation=<rot 0 0 90>}]
 dyn_right_bowB_01 [{shapeType=Capsule, length=2.0495, radius=0.9427, center=<1 0 0>, rotation=<rot 0 0 90>}]
 dyn_right_bowB_02 [{shapeType=Capsule, length=2.3311, radius=2.0007, center=<1 0 0>, rotation=<rot 0 0 90>}]
 dyn_right_bowC_01 [{shapeType=Capsule, length=1.25, radius=1.75, center=<4.1266 0 0>, rotation=<rot 84.9966 180 180>}]
 dyn_right_bowC_02 [{shapeType=Capsule, length=4, radius=2, center=<4.1611 0 1.1264>, rotation=<rot 74.9955 -179.9999 174.9999>}]
}
constraints [
 {
  bone1Name=dyn_heart
  bone2Name=Melee_TwoHanded_Handle
  swing1LimitDegrees=45
  swing1Motion=Free
  swing2LimitDegrees=45
  swing2Motion=Free
  twistLimitDegrees=45
  twistMotion=Free
 }
 {bone1Name=dyn_right_bowC_01, bone2Name=dyn_heart, swing1LimitDegrees=25, swing2LimitDegrees=60, twistLimitDegrees=2}
 {bone1Name=dyn_right_bowC_02, bone2Name=dyn_right_bowC_01, swing1LimitDegrees=45, swing2LimitDegrees=45, twistLimitDegrees=2}
 {bone1Name=dyn_left_bowC_01, bone2Name=dyn_heart, swing1LimitDegrees=25, swing2LimitDegrees=60, twistLimitDegrees=2}
 {bone1Name=dyn_left_bowC_02, bone2Name=dyn_left_bowC_01, swing1LimitDegrees=45, swing2LimitDegrees=45, twistLimitDegrees=2}
 {bone1Name=dyn_left_bowC_03, bone2Name=dyn_left_bowC_02, swing1LimitDegrees=45, swing2LimitDegrees=45, twistLimitDegrees=2}
 {bone1Name=dyn_left_bowA_01, bone2Name=dyn_heart, swing1LimitDegrees=35, swing2LimitDegrees=35, twistLimitDegrees=2}
 {bone1Name=dyn_left_bowA_02, bone2Name=dyn_left_bowA_01, swing1LimitDegrees=35, swing2LimitDegrees=35, twistLimitDegrees=2}
 {bone1Name=dyn_right_bowA_01, bone2Name=dyn_heart, swing1LimitDegrees=35, swing2LimitDegrees=35, twistLimitDegrees=2}
 {bone1Name=dyn_right_bowA_02, bone2Name=dyn_right_bowA_01, swing1LimitDegrees=35, swing2LimitDegrees=35, twistLimitDegrees=2}
 {bone1Name=dyn_left_bowB_01, bone2Name=dyn_heart, swing1LimitDegrees=20, swing2LimitDegrees=20, twistLimitDegrees=2}
 {bone1Name=dyn_left_bowB_02, bone2Name=dyn_left_bowB_01, swing1LimitDegrees=20, swing2LimitDegrees=20, twistLimitDegrees=2}
 {bone1Name=dyn_right_bowB_01, bone2Name=dyn_heart, swing1LimitDegrees=20, swing2LimitDegrees=20, twistLimitDegrees=2}
 {bone1Name=dyn_right_bowB_02, bone2Name=dyn_right_bowB_01, swing1LimitDegrees=20, swing2LimitDegrees=20, twistLimitDegrees=2}
]
